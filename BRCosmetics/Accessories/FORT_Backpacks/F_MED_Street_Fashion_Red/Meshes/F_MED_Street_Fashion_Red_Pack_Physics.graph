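# Physics asset for the Street Fashion Red backpack cosmetic, providing collision bodies and a joint for the bone named dyn_knife.
# A large box on root covers the pack, a small box sits on parent_dynamics, and a thin capsule represents the knife.
# The only constraint joins dyn_knife to root with tight swing and twist limits of a few degrees, allowing just slight wobble.
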PhysicsAsset /BRCosmetics/Accessories/FORT_Backpacks/F_MED_Street_Fashion_Red/Meshes/F_MED_Street_Fashion_Red_Pack_Physics
bodies {
 dyn_knife [{shapeType=Capsule, length=22.0015, radius=1.3, center=<-1.88 -0.5623 -0.2436>, rotation=<rot 0 102.8882 -90.3327>}]
 parent_dynamics [{shapeType=Box, extentX=2.125, extentY=2.125, extentZ=2.125}]
 root [{shapeType=Box, extentX=7.5, extentY=16.5, extentZ=29, center=<-5.7385 -7.0964 -4.6335>}]
}
constraints [{bone1Name=dyn_knife, bone2Name=root, swing1LimitDegrees=6, swing2LimitDegrees=2, twistLimitDegrees=3}]
