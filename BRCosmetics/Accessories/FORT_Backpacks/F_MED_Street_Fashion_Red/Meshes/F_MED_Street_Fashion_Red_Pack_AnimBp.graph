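# Animation Blueprint for the Street Fashion Red backpack cosmetic, adding secondary motion to its headphones, zippers and physics-asset bodies.
# A RigidBody node simulates the pack's physics asset from the ref pose in root-bone space, then feeds chained AnimDynamics nodes for each headphone and zipper bone.
# Each update lowers the earphone and headphone simulation alphas during skydive dive or float, while the zippers stay fully simulated.
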AnimBlueprint /BRCosmetics/Accessories/FORT_Backpacks/F_MED_Street_Fashion_Red/Meshes/F_MED_Street_Fashion_Red_Pack_AnimBp
target_skeleton=&/F_MED_Street_Fashion_Red_Pack_Skeleton
graphs [
 AnimGraph {nodes [
   n1 = RigidBody(ComponentPose=Local To Component(LocalPose=Local Space Ref Pose())) {
    bClampLinearTranslationLimitToRefPose=true
    bOverrideWorldGravity=true
    overridePhysicsAsset=&/F_MED_Street_Fashion_Red_Pack_Physics
    simulationSpace=BaseBoneSpace
    alphaBoolBlend {blendOption=Linear}
    baseBoneRef {boneName=root}
    componentAppliedLinearAccClamp=<6000 6000 6000>
    componentLinearAccScale=<0.25 0.25 0.5>
    componentLinearVelScale=<0.25 0.25 0.5>
    overrideWorldGravity {z=-980}
   }
   n2 = Get Earphone_Alpha() -> (Earphone_Alpha:double)
   n3 = Anim Dynamics - Bone: dyn_headphone_01(ComponentPose=@n1, Alpha=@n2, ExternalForce="0,0,0") {
    angularDampingOverride=0.9999
    bGravityOverrideInSimSpace=true
    bOverrideAngularDamping=true
    bOverrideLinearDamping=true
    linearDampingOverride=0.9999
    alphaBoolBlend {blendOption=Linear}
    boundBone {boneName=dyn_headphone_01}
    componentAppliedLinearAccClamp=<7000 7000 7000>
    componentLinearAccScale=<0.5 0.5 0.5>
    componentLinearVelScale=<0.5 0.5 0.5>
    physicsBodyDefinitions [{
      sphereCollisionRadius=0
      boundBone {boneName=dyn_headphone_01}
      boxExtents=<8 2 5>
      constraintSetup {angularLimitsMax=<0 10 45>, angularLimitsMin=<0 -5 -45>, angularTarget=<0 0 0>}
      localJointOffset=<3 0 0>
     }]
    relativeSpaceBone {boneName=root}
   }
   n4 = Anim Dynamics - Bone: dyn_headphone_02(ComponentPose=@n3, Alpha=@n2, ExternalForce="0,0,0") {
    angularDampingOverride=0.9999
    bGravityOverrideInSimSpace=true
    bOverrideAngularDamping=true
    bOverrideLinearDamping=true
    linearDampingOverride=0.9999
    alphaBoolBlend {blendOption=Linear}
    boundBone {boneName=dyn_headphone_02}
    componentAppliedLinearAccClamp=<7000 7000 7000>
    componentLinearAccScale=<0.5 0.5 0.5>
    componentLinearVelScale=<0.5 0.5 0.5>
    physicsBodyDefinitions [{
      sphereCollisionRadius=0
      boundBone {boneName=dyn_headphone_02}
      boxExtents=<8 2 5>
      constraintSetup {angularLimitsMax=<0 10 45>, angularLimitsMin=<0 -5 -45>, angularTarget=<0 0 0>}
      localJointOffset=<3 0 0>
     }]
    relativeSpaceBone {boneName=root}
   }
   n5 = Anim Dynamics - Bone: dyn_headphones(ComponentPose=@n4, Alpha=Get Skydive_Alpha() -> (Skydive_Alpha:double), ExternalForce="0,0,0") {
    angularDampingOverride=0.9999
    bGravityOverrideInSimSpace=true
    bOverrideAngularDamping=true
    bOverrideLinearDamping=true
    linearDampingOverride=0.9999
    alphaBoolBlend {blendOption=Linear}
    boundBone {boneName=dyn_headphones}
    componentAppliedLinearAccClamp=<600 600 600>
    componentLinearAccScale=<0.25 0.25 0.25>
    componentLinearVelScale=<0.25 0.25 0.25>
    physicsBodyDefinitions [{
      sphereCollisionRadius=0
      boundBone {boneName=dyn_headphones}
      boxExtents=<12 12 2>
      constraintSetup {angularLimitsMax=<0 15 45>, angularLimitsMin=<0 -75 -45>, angularTarget=<0 0 0>}
      localJointOffset=<5 0 0>
     }]
    relativeSpaceBone {boneName=root}
   }
   n6 = Anim Dynamics - Bone: dyn_zipper_01(ComponentPose=@n5, Alpha=1, ExternalForce="0,0,0") {
    angularDampingOverride=0.999
    bGravityOverrideInSimSpace=true
    bOverrideAngularDamping=true
    gravityScale=0.75
    alphaBoolBlend {blendOption=Linear}
    boundBone {boneName=dyn_zipper_01}
    componentAppliedLinearAccClamp=<9000 9000 9000>
    componentLinearAccScale=<0.25 0.25 0.25>
    componentLinearVelScale=<0.25 0.25 0.25>
    physicsBodyDefinitions [{
      sphereCollisionRadius=0
      boundBone {boneName=dyn_zipper_01}
      boxExtents=<4.5 2 1.2>
      constraintSetup {angularLimitsMax=<0 4 15>, angularLimitsMin=<0 -20 -15>, angularTarget=<0 0 0>}
      localJointOffset=<1.3 0 0>
     }]
    relativeSpaceBone {boneName=root}
   }
   n7 = Anim Dynamics - Bone: dyn_zipper_02(ComponentPose=@n6, Alpha=1, ExternalForce="0,0,0") {
    angularDampingOverride=0.99
    bGravityOverrideInSimSpace=true
    bOverrideAngularDamping=true
    bOverrideLinearDamping=true
    gravityScale=0.75
    linearDampingOverride=0.99
    alphaBoolBlend {blendOption=Linear}
    boundBone {boneName=dyn_zipper_02}
    componentAppliedLinearAccClamp=<9000 9000 9000>
    componentLinearAccScale=<0.25 0.25 0.25>
    componentLinearVelScale=<0.25 0.25 0.25>
    physicsBodyDefinitions [{
      sphereCollisionRadius=0
      boundBone {boneName=dyn_zipper_02}
      boxExtents=<4.5 2 1.2>
      constraintSetup {angularLimitsMax=<0 4 15>, angularLimitsMin=<0 -20 -15>, angularTarget=<0 0 0>}
      localJointOffset=<1.3 0 0>
     }]
    relativeSpaceBone {boneName=root}
   }
   Output Pose(Result=Component To Local(ComponentPose=@n7)) {layerGroup=None}
  ]}
 EventGraph {nodes [
   Event BlueprintUpdateAnimation()
   n1 = Get bIsSkydiveDiveMode()
   n2 = Get bIsSkydiveFloating()
   > Set Skydive_Alpha(Skydive_Alpha:double=SelectFloat(A=0.25, B=0.5, bPickA=OR Boolean(A=@n2, B=@n1))) -> (Output_Get:double)
   n3 = Get bIsSkydiveDiveMode()
   n4 = Get bIsSkydiveFloating()
   > Set Earphone_Alpha(Earphone_Alpha:double=SelectFloat(A=0.25, B=0.4, bPickA=OR Boolean(A=@n4, B=@n3))) -> (Output_Get:double)
   TryGetPawnOwner()
  ]}
]
variables [
 Earphone_Alpha {type=number}
 Skydive_Alpha {type=number}
]
